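# Niagara system for the prop-spawn effect of the MicstandFlip mic emote in SparksCosmetics.
# It holds four emitters for sparkles, airglow and two smoke layers, whose LWE names suggest lightweight (stateless) emitters.
# System scripts use SystemState and a spawn-time variable assignment, with the FNTimeOfDayColor dynamic input available for time-of-day color.
# It exposes a NormalizedNotifyProgress float, apparently for animation-notify timing.
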NiagaraSystem /SparksCosmetics/Emotes/Mic/MicstandFlip/Effects/NS_Emotes_MicstandFlip_PropSpawn
dependencies {
 usedDataInterfaces []
 usedDynamicInputs [&FNTimeOfDayColor]
 usedModules [
  &self:SystemSpawnScript.SystemScriptSource.SystemScriptGraph.NiagaraNodeAssignment_0.SetVariables_#1
  &SystemState
 ]
}
emitters {
 AirglowLWE {rendererClasses [], renderers []}
 SmokeLWE {rendererClasses [], renderers []}
 SmokeLWE001 {rendererClasses [], renderers []}
 SparklesLWE {rendererClasses [], renderers []}
}
summary {
 systemName=NS_Emotes_MicstandFlip_PropSpawn
 emitters [
  {emitterName=SparklesLWE, rendererClasses []}
  {emitterName=AirglowLWE, rendererClasses []}
  {emitterName=SmokeLWE, rendererClasses []}
  {emitterName=SmokeLWE001, rendererClasses []}
 ]
 userVariables [User.NormalizedNotifyProgress {type=NiagaraFloat, defaultValue=0, description=""}]
}
